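# Movement-mode anim layer for the Fortnite player skeleton that animates a DBNO "ragdoll" tumble through its FullBodyOverride layer.
# A state machine starts in a downhill loop, switches to a fall loop when off the ground, and passes through the incoming pose on OutroActive.
# Its output gets spine roll offsets, an additive crouch lean and a montage slot, then Dead Blending.
# Animation is sequence-driven, without physics nodes.
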AnimBlueprint /TrendConcentrate/Assets/Animation/ABP_TrendConcentrate_DBNORagdoll
target_skeleton=&Fortnite_M_Avg_Player_Skeleton
graphs [
 AnimGraph {nodes [Output Pose() {layerGroup=None}]}
 EventGraph {links [], nodes []}
 MovementMode_AimOffset {nodes [
   Input Pose() {name=InAimOffsetsPose}
   Input Pose() {name=InSourcePosePreAimOffsets}
   Output Pose() {layerGroup=MovementMode}
  ]}
 MovementMode_BaseLocomotion {nodes [
   Input Pose() {name=InPoseBaseLocomotion}
   Output Pose() {layerGroup=MovementMode}
  ]}
 MovementMode_FinalPoseOverride {nodes [
   Input Pose() {name=InFinalPose}
   Output Pose() {layerGroup=MovementMode}
  ]}
 MovementMode_FullBodyOverride {nodes [
   Output Pose(Result=Dead Blending!(Source:Pose Link Structure=Use cached pose 'PreRootOffset'()) -> (Pose:Pose Link Structure) {is=title AnimGraphNode_DeadBlending}) {layerGroup=MovementMode}
   Save cached pose 'CacheMovementFullbody'(Pose=Input Pose() {name=InputPoseFullBody}) {cacheName=CacheMovementFullbody}
   n1 = Blendspace Player 'Crouch_Lean'!(X=0, PlayRate:float=1) -> (Pose) {
    blendSpace=/Game/Animation/Game/MainPlayer/Locomotion/Medium/Male/Crouch/Lean/Crouch_Lean
    groupName=Locomotion
    groupRole=AlwaysFollower
    method=SyncGroup
   }
   AnimGraphNode_StateMachine_194 = DBNO() {is=title AnimGraphNode_StateMachine, bSkipFirstUpdateTransition=false}
   n2 = "Transform (Modify) Bone - Bone: Spine_02"!(
    ComponentPose=Local To Component(LocalPose=@AnimGraphNode_StateMachine_194)
    bAlphaBoolEnabled:bool=True
    Translation="0,0,0"
    Rotation="0,0,-8"
    Scale="1,1,1"
   ) -> (Pose) {alphaInputType=Bool, rotationMode=BMM_Additive, alphaBoolBlend {blendOption=Linear}, boneToModify {boneName=Spine_02}, rotation {roll=-8}}
   n3 = "Transform (Modify) Bone - Bone: spine_05"!(ComponentPose=@n2, bAlphaBoolEnabled:bool=True, Translation="0,0,0", Rotation="0,0,4", Scale="1,1,1") -> (Pose) {alphaInputType=Bool, rotationMode=BMM_Additive, alphaBoolBlend {blendOption=Linear}, boneToModify {boneName=spine_05}, rotation {roll=4}}
   n4 = Component To Local(ComponentPose=@n3)
   Save cached pose 'PreRootOffset'(Pose=Slot 'TDiA.Custom_01'(Source=Apply Additive(Base=@n4, Additive=@n1) {alphaBoolBlend {blendOption=Linear}}) {slotName=TDiA.Custom_01}) {cacheName=PreRootOffset}
  ]}
 DBNO {
  entry=DBNO Ragdoll Loop
  path=MovementMode_FullBodyOverride.AnimGraphNode_StateMachine_194.DBNO
  ids=id
  nodes [
   Fall Alias() {is=title AnimStateAliasNode, also {aliases [DBNO Ragdoll Loop]}}
   To Idle() {is=title AnimStateAliasNode, also {aliases [DBNO Fall, DBNO Ragdoll Loop]}}
   AnimStateNode_0 = DBNO Ragdoll Loop() {
    is=title AnimStateNode
    stateEntered {guid=#1}
    stateFullyBlended {guid=#2}
    stateLeft {guid=#3}
   }
   AnimStateNode_1 = Idle() {
    stateEntered {guid=#4}
    stateFullyBlended {guid=#5}
    stateLeft {guid=#6}
   }
   AnimStateNode_2 = DBNO Fall() {is=title AnimStateNode}
   AnimStateTransitionNode_2 = AnimStateTransitionNode() {bAllowInertializationForSelfTransitions=true, crossfadeDuration=0}
   AnimStateTransitionNode_3 = AnimStateTransitionNode() {bAllowInertializationForSelfTransitions=true}
   AnimStateTransitionNode_4 = AnimStateTransitionNode() {bAllowInertializationForSelfTransitions=true}
  ]
  links [
   {
    dst=DBNO Ragdoll Loop
    rule=MovementMode_FullBodyOverride.AnimGraphNode_StateMachine_194.DBNO.AnimStateTransitionNode_4.Transition
    src=DBNO Fall
    via=AnimStateTransitionNode_4
   }
   {
    dst=DBNO Fall
    rule=MovementMode_FullBodyOverride.AnimGraphNode_StateMachine_194.DBNO.AnimStateTransitionNode_3.Transition
    src=Fall Alias
    via=AnimStateTransitionNode_3
   }
   {
    dst=Idle
    rule=MovementMode_FullBodyOverride.AnimGraphNode_StateMachine_194.DBNO.AnimStateTransitionNode_2.Transition
    src=To Idle
    via=AnimStateTransitionNode_2
   }
  ]
 }
 DBNO Ragdoll Loop {
  path=MovementMode_FullBodyOverride.AnimGraphNode_StateMachine_194.DBNO.AnimStateNode_0.DBNO Ragdoll Loop
  nodes [
   n1 = Sequence Player 'LBNO_Downhill_Loop_CMM'!(PlayRate:float=1) -> (Pose) {
    bOverridePositionWhenJoiningSyncGroupAsLeader=true
    groupName=Roll
    groupRole=AlwaysLeader
    method=SyncGroup
    sequence=&/LBNO_Downhill_Loop_CMM
   }
   Output Animation Pose(Result=Mirror with MDT_Player_Default!(Source:Pose Link Structure=@n1, bMirror:bool=false) -> (Pose:Pose Link Structure) {is=title AnimGraphNode_Mirror, mirrorDataTable=/Game/Characters/Player/DataTables/MDT_Player_Default})
  ]
 }
 Idle {
  path=MovementMode_FullBodyOverride.AnimGraphNode_StateMachine_194.DBNO.AnimStateNode_1.Idle
  nodes [Output Animation Pose(Result=Use cached pose 'CacheMovementFullbody'())]
 }
 DBNO Fall {
  path=MovementMode_FullBodyOverride.AnimGraphNode_StateMachine_194.DBNO.AnimStateNode_2.DBNO Fall
  nodes [Output Animation Pose(Result=Sequence Player 'DBNO_Fall_Loop'() {sequence=/Game/Animation/Game/MainPlayer/Locomotion/Medium/Male/DBNO/DBNO_Fall_Loop})]
 }
 Transition {
  path=MovementMode_FullBodyOverride.AnimGraphNode_StateMachine_194.DBNO.AnimStateTransitionNode_2.Transition
  nodes [Result(bCanEnterTransition=Get OutroActive() -> (OutroActive:bool))]
 }
 Transition {
  path=MovementMode_FullBodyOverride.AnimGraphNode_StateMachine_194.DBNO.AnimStateTransitionNode_3.Transition
  nodes [Result(bCanEnterTransition=NOT Boolean(A=Get bIsOnGround() -> (bIsOnGround:bool)))]
 }
 Transition {path=MovementMode_FullBodyOverride.AnimGraphNode_StateMachine_194.DBNO.AnimStateTransitionNode_4.Transition, nodes [Result()]}
 MovementMode_InAirOverride {nodes [
   Input Pose() {name=InSource_Pose}
   Input Pose() {name=InAirPose}
   Output Pose() {layerGroup=MovementMode}
  ]}
 MovementMode_LocomotionAdditives {nodes [
   Input Pose() {name=InLocomotionAdditivesPose}
   Output Pose() {layerGroup=MovementMode}
  ]}
 MovementMode_LowerBodyOverride {nodes [
   Input Pose() {name=InLowerBodyPose}
   Input Pose() {name=InSourcePose}
   Output Pose() {layerGroup=MovementMode}
  ]}
 MovementMode_OffScreen {nodes [
   Input Pose() {name=InPoseOffscreen}
   Output Pose() {layerGroup=MovementMode}
  ]}
 MovementMode_SourcePose {nodes [
   Input Pose() {name=InSourcePoseBase}
   Output Pose() {layerGroup=MovementMode}
  ]}
]
variables [OutroActive {type=boolean}]
