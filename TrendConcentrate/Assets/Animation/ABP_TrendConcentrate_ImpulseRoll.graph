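# Animation Blueprint for the TrendConcentrate impulse roll on the Fortnite player skeleton, supplying MovementMode animation layers.
# Only the FinalPoseOverride layer has logic: the incoming final pose passes through the TDiA.Custom_01 slot into a cached pose.
# That pose blends with a root-bone copy yawed by RootOffset, weighted by RootOffsetAlpha, then feeds a Dead Blending node.
# Other layers and the event graph are unwired, and no variables are declared here.
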AnimBlueprint /TrendConcentrate/Assets/Animation/ABP_TrendConcentrate_ImpulseRoll
target_skeleton=&Fortnite_M_Avg_Player_Skeleton
graphs [
 AnimGraph {nodes [Output Pose()]}
 EventGraph {links [], nodes []}
 MovementMode_AimOffset {nodes [
   Input Pose() {name=InAimOffsetsPose}
   Input Pose() {name=InSourcePosePreAimOffsets}
   Output Pose() {layerGroup=MovementMode}
  ]}
 MovementMode_BaseLocomotion {nodes [
   Input Pose() {name=InPoseBaseLocomotion}
   Output Pose() {layerGroup=MovementMode}
  ]}
 MovementMode_FinalPoseOverride {nodes [
   n1 = Rotate Root Bone!(
    BasePose:Pose Link Structure=Use cached pose 'FullbodyOverrideInputCache'()
    Pitch:float=0
    Yaw:float=Get RootOffset() -> (RootOffset:double)
    MeshToComponent:Rotator="0,0,0"
   ) -> (Pose:Pose Link Structure) {is=title AnimGraphNode_RotateRootBone}
   n2 = Use cached pose 'FullbodyOverrideInputCache'()
   n3 = Get RootOffsetAlpha() -> (RootOffsetAlpha:float)
   Output Pose(Result=Dead Blending!(Source:Pose Link Structure=Two Way Blend(A=@n2, B=@n1, Alpha=@n3) {blendNode {bAlwaysUpdateChildren=true}}) -> (Pose:Pose Link Structure) {is=title AnimGraphNode_DeadBlending}) {layerGroup=MovementMode}
   Save cached pose 'FullbodyOverrideInputCache'(Pose=Slot 'TDiA.Custom_01'(Source=Input Pose() {name=InFinalPose}) {slotName=TDiA.Custom_01}) {cacheName=FullbodyOverrideInputCache}
  ]}
 MovementMode_FullBodyOverride {nodes [
   Input Pose() {name=InputPoseFullBody}
   Output Pose() {layerGroup=MovementMode}
  ]}
 MovementMode_InAirOverride {nodes [
   Input Pose() {name=InSource_Pose}
   Input Pose() {name=InAirPose}
   Output Pose() {layerGroup=MovementMode}
  ]}
 MovementMode_LocomotionAdditives {nodes [
   Input Pose() {name=InLocomotionAdditivesPose}
   Output Pose() {layerGroup=MovementMode}
  ]}
 MovementMode_LowerBodyOverride {nodes [
   Input Pose() {name=InLowerBodyPose}
   Input Pose() {name=InSourcePose}
   Output Pose() {layerGroup=MovementMode}
  ]}
 MovementMode_OffScreen {nodes [
   Input Pose() {name=InPoseOffscreen}
   Output Pose() {layerGroup=MovementMode}
  ]}
 MovementMode_SourcePose {nodes [
   Input Pose() {name=InSourcePoseBase}
   Output Pose() {layerGroup=MovementMode}
  ]}
]
variables []
